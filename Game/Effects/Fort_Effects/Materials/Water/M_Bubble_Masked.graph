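# Masked, lit Niagara particle material that draws wobbling bubbles.
# Panned noise, randomized per particle, distorts UVs feeding a circular gradient; dynamic parameters control wobble strength, speed, emissive and contrast.
# The gradient shapes temporally dithered opacity, a powered edge rim and sphere-like warped UVs that sample a mirrored reflection highlight and sphere normal map.
# Particle color tints the result, and a switchable camera-distance offset shrinks bubbles near the camera.
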Material /Game/Effects/Fort_Effects/Materials/Water/M_Bubble_Masked
graph {
 properties {
  bAutomaticallySetUsageInEditor=false
  bUsedWithNiagaraMeshParticles=true
  bUsedWithNiagaraSprites=true
  bUsedWithStaticMesh=false
  blendMode=BLEND_Masked
  pixelDepthOffsetMode=PDOM_Legacy
  usedShadingModels=Default Lit
 }
 nodes [
  n1 = Constant2Vector() {g=0.5, r=0.5}
  n2 = MaterialFunctionCall(Float2=TextureCoordinate()) {materialFunction=&BreakOutFloat2Components}
  n3 = ScalarParameter() {defaultValue=0.3, parameterName=Distort_U_Tile}
  n4 = ScalarParameter() {defaultValue=0.3, parameterName=Distort_V_Tile}
  n5 = Add(
   A=AppendVector(A=Multiply(A=@n2.R, B=@n3), B=Multiply(A=@n2.G, B=@n4))
   B=ShadingPathSwitch(Default=ParticleRandom(), Mobile=Constant()) {desc=Migrated from FeatureLevelSwitch}
  )
  n6 = DynamicParameter() {defaultValue=<rgba 1 1 1 -1>, paramNames [WibbMult, WibbSpeedMult, Emissive, BaseColorContrast]}
  n7 = Multiply(A=Time(), B=@n6.WibbSpeedMult)
  n8 = AppendVector(
   A=ScalarParameter() {defaultValue=0.02, parameterName=Distort_U_PanSpeed}
   B=ScalarParameter() {defaultValue=0.15, parameterName=Distort_V_PanSpeed}
  )
  n9 = MaterialFunctionCall() {materialFunction=&LinearGradient}
  n10 = LinearInterpolate(
   B=TextureSample(UVs=Panner(Coordinate=@n5, Time=@n7, Speed=@n8) {speedX=0.02, speedY=0.15}) {samplerType=SAMPLERTYPE_Grayscale, texture=&LowResBlurredNoise}.G
   Alpha=LinearInterpolate(A=SquareRoot(None=@n9.UGradient), Alpha=ScalarParameter() {defaultValue=1, parameterName=Lerp_GradientBlendDistortNoise})
  )
  LinearInterpolate(
   A=ComponentMask(None=VectorParameter() {parameterName=SphereMaskPosition, defaultValue=<rgba 0.5 0.5 0 0>}.RGB) {g=true, r=true}
   B=@n1
   Alpha=@n10
  )
  n11 = ParticleColor()
  n12 = MaterialFunctionCall(Float2=TextureCoordinate()) {materialFunction=&BreakOutFloat2Components}
  n13 = ScalarParameter() {defaultValue=1, parameterName=Base_U_Tile}
  n14 = ScalarParameter() {defaultValue=1, parameterName=Base_V_Tile}
  n15 = ScalarParameter() {defaultValue=4, parameterName=WibbStrength}
  n16 = Add(A=TextureCoordinate()) {constB=-0.5}
  n17 = Add(
   A=AppendVector(A=Multiply(A=@n12.R, B=@n13), B=Multiply(A=@n12.G, B=@n14))
   B=Multiply(A=Multiply(A=@n6.WibbMult, B=@n15), B=Multiply(A=@n16, B=@n10))
  )
  n18 = MaterialFunctionCall!(UVs=@n17, CenterPosition=@n1, Radius=ScalarParameter() {defaultValue=0.5, parameterName=SphereMaskRadius}) {materialFunction=/Engine/Functions/Engine_MaterialFunctions01/Gradient/SphereGradient-2D}
  n19 = ScalarParameter() {defaultValue=0.11, parameterName=MinOpacity}
  n20 = Multiply(A=Multiply(A=@n18.Result 0-1, B=LinearInterpolate(B=@n19, Alpha=@n18.Result 0-1) {constA=1})) {constB=4}
  MaterialFunctionCall(Alpha Threshold=Multiply(A=@n11.A, B=@n20)) {materialFunction=&DitherTemporalAA}
  n21 = ScalarParameter() {defaultValue=1, parameterName=Highlight Strength}
  n22 = Add(A=@n17) {constB=-0.5}
  n23 = OneMinus(None=@n18.Result 0-1)
  n24 = LinearInterpolate(Alpha=@n23) {constA=0.1}
  n25 = Add(A=Multiply(A=Multiply(A=@n22, B=@n24)) {constB=2}) {constB=0.5}
  n26 = TextureSample(UVs=OneMinus(None=@n25)) {texture=/Game/Items/Art_noLOD/Textures/T_Items_MineralPowder_Glass_Reflection}
  n27 = ScalarParameter() {defaultValue=0.25, parameterName=Secondary Highlight Strength}
  n28 = Multiply(
   A=Add(A=TextureSample(UVs=@n25) {texture=&T_Items_MineralPowder_Glass_Reflection}.G, B=Multiply(A=@n26.G, B=@n27))
   B=@n20
  )
  n29 = Multiply(A=@n23, B=@n23)
  n30 = Multiply(A=@n29, B=@n29)
  n31 = Multiply(A=@n30, B=@n30)
  n32 = ScalarParameter() {defaultValue=3, parameterName=Edge Strength}
  n33 = Multiply(A=Add(A=Multiply(A=@n21, B=@n28), B=Multiply(A=@n31, B=@n32)), B=@n11.RGB)
  MaterialFunctionCall(
   In=@n33
   Contrast=DynamicParameter() {defaultValue=<rgba 1 1 1 -1>, paramNames [WibbMult, WibbSpeedMult, Emissive, BaseColorContrast]}.BaseColorContrast
  ) {materialFunction=&CheapContrast}
  Multiply(
   A=MaterialFunctionCall(Emissive=@n33) {materialFunction=&MF_FX_EmissiveAdjustments}.Result
   B=DynamicParameter() {defaultValue=<rgba 1 1 1 -1>, paramNames [WibbMult, WibbSpeedMult, Emissive, BaseColorContrast]}.Emissive
  )
  ScalarParameter() {defaultValue=0.2, parameterName=EmissiveMult}
  ScalarParameter() {defaultValue=-1, parameterName=BaseColorContrast}
  n34 = MaterialFunctionCall(Distance=ScalarParameter() {defaultValue=100, parameterName=CamFadeDist}) {materialFunction=&Near_Camera_Fade}
  n35 = Transform(None=AppendVector(A=TextureCoordinate(), B=Constant()))
  n36 = ComponentMask(None=ParticleSize()) {r=true}
  StaticSwitchParameter(True=Multiply(A=LinearInterpolate(Alpha=@n34.Result) {constA=-1, constB=0}, B=Multiply(A=@n35, B=@n36)), False=Constant()) {defaultValue=true, parameterName=ShrinkBasedOnCameraDistance}
  TextureSample(UVs=MaterialFunctionCall(UVs=@n25, Texture Scale=Constant() {r=1.5}) {materialFunction=&ScaleUVsByCenter}.UVs) {samplerType=SAMPLERTYPE_Normal, texture=&T_Sphere_N}
 ]
}
