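# Translucent lit mesh-particle material for foamy water wakes, apparently around a cube's edges.
# Base color combines sphere-mapped reflection, time-of-day water tint modulated by four-way panning caustics, and noise-plus-bubble textures.
# A UV box-edge ramp shapes particle-colored emissive; vertex red masks depth-faded opacity and noise-driven world position offset, which switches between box ramp and sphere mask.
# Dynamic parameters drive fade, foam height, spread and power, with optional near-camera fading.
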Material /Game/Effects/Fort_Effects/Materials/Water/M_CubeEdgeWake
graph {
 properties {
  bUsedWithMeshParticles=true
  bUsedWithNiagaraMeshParticles=true
  blendMode=BLEND_Translucent
  pixelDepthOffsetMode=PDOM_Legacy
  translucencyLightingMode=TLM_VolumetricPerVertexNonDirectional
  usedShadingModels=Default Lit
 }
 nodes [
  n1 = Transform(None=ReflectionVectorWS(CustomWorldNormal=TextureSample() {samplerType=SAMPLERTYPE_Normal, texture=/Game/Effects/Fort_Effects/Textures/Water/T_MaskeSplashes_Bubbles_N}.RGB)) {transformSourceType=TRANSFORMSOURCE_World, transformType=TRANSFORM_View}
  n2 = Multiply(
   A=ConstantBiasScale(None=ComponentMask(None=@n1) {g=true, r=true})
   B=ComponentMask(None=VectorParameter() {parameterName=CubeMapOffset, defaultValue=<rgba 1 1 0 1>}.RGB) {g=true, r=true}
  )
  n3 = Multiply(
   A=Multiply(
    A=TextureSampleParameter2D(UVs=@n2) {parameterName=SphereReflectionMap, texture=/Game/Items/Art_noLOD/Textures/T_Items_MineralPowder_Glass_Reflection}.RGB
    B=ScalarParameter() {defaultValue=16, group=reflection, parameterName=Reflection Multiplier}
   )
   B=ComponentMask(None=MaterialFunctionCall() {materialFunction=/Engine/Functions/GetPostProcessFunctions/GetAmbientCubemapTint}.AmbientCubemapTint) {b=true, g=true, r=true}
  )
  n4 = VectorParameter() {parameterName=Base Color Add, defaultValue=<rgba 1 1 1 0>}
  n5 = Multiply(A=TextureCoordinate(), B=ScalarParameter() {defaultValue=3, parameterName=Color Variation Tile})
  n6 = TextureObjectParameter() {parameterName=Color Noise Texture, texture=&T_ShieldCaustics}
  n7 = TextureSample(UVs=Panner(Coordinate=@n5, Speed=Constant2Vector() {g=0.1, r=0.1}), Tex=@n6) {texture=&T_ShieldCaustics}
  n8 = Add(A=@n5, B=Constant2Vector() {g=0.3548, r=0.4181})
  n9 = Constant2Vector() {g=-0.1, r=0.1}
  n10 = TextureSample(UVs=Multiply(A=Panner(Coordinate=@n8, Speed=@n9)) {constB=0.7143}, Tex=@n6) {texture=&T_ShieldCaustics}
  n11 = Add(A=@n5, B=Constant2Vector() {g=0.1484, r=0.8649})
  n12 = Constant2Vector() {g=0.1, r=-0.1}
  n13 = TextureSample(UVs=Multiply(A=Panner(Coordinate=@n11, Speed=@n12)) {constB=1.186}, Tex=@n6) {texture=&T_ShieldCaustics}
  n14 = Add(A=@n5, B=Constant2Vector() {g=0.7516, r=0.6513})
  n15 = Constant2Vector() {g=-0.1, r=-0.1}
  n16 = TextureSample(UVs=Multiply(A=Panner(Coordinate=@n14, Speed=@n15)) {constB=1.1803}, Tex=@n6) {texture=&T_ShieldCaustics}
  n17 = VectorParameter() {parameterName=Color Texture mask Options, channelNames {a="", b=Add, g=Multiply, r=Power}, defaultValue=<rgba 1 1 0 0>}
  n18 = Power(Base=Max(A=Max(A=@n7.R, B=@n10.G), B=Max(A=@n13.R, B=@n16.B)), Exp=@n17.Power)
  n19 = Clamp(None=Add(A=Multiply(A=@n18, B=@n17.Multiply), B=@n17.Add))
  n20 = LinearInterpolate(A=@n4.RGB, B=VectorParameter() {parameterName=Base Color Add 2, defaultValue=<rgba 1 1 1 0>}.RGB, Alpha=@n19)
  n21 = ScalarParameter() {parameterName=Use Second COlor}
  n22 = MaterialFunctionCall(
   Color_Morning=VectorParameter() {parameterName=WaterColorMorning, defaultValue=<rgba 0.3909 0.3217 0.5 1>}.RGB
   Color_Day=VectorParameter() {parameterName=WaterColorDay, defaultValue=<rgba 0.1779 0.3005 0.6308 1>}.RGB
   Color_Evening=VectorParameter() {parameterName=WaterColorEve, defaultValue=<rgba 0.008 0.0561 0.314 1>}.RGB
   Color_Night=VectorParameter() {parameterName=WaterColorNight, defaultValue=<rgba 0.3255 0.116 0.5573 1>}.RGB
  ) {materialFunction=&Color_Time_of_Day}
  n23 = LinearInterpolate(A=@n22.Result, Alpha=@n19)
  n24 = TextureCoordinate()
  n25 = ScalarParameter() {defaultValue=2, parameterName=Caustic_Scale}
  n26 = TextureSample(UVs=Panner(Coordinate=Multiply(A=@n24, B=@n25)) {speedY=0.05}) {samplerType=SAMPLERTYPE_LinearColor, texture=/Game/Characters/Player/Male/Medium/Bodies/M_MED_SRDrift/Textures/T_ElectricNoise}
  n27 = TextureCoordinate()
  n28 = ScalarParameter() {defaultValue=5, parameterName=BubbleScale}
  n29 = Add(A=@n26.RGB, B=TextureSample(UVs=Multiply(A=@n27, B=@n28)) {texture=/Game/Effects/Fort_Effects/Textures/Water/T_MaskeSplashes_Bubbles1_A}.RGB)
  Add(A=Add(A=Clamp(None=@n3), B=LinearInterpolate(A=@n4.RGB, B=@n20, Alpha=@n21)), B=Multiply(A=LinearInterpolate(A=@n22.Result, B=@n23), B=@n29))
  n30 = DynamicParameter() {paramNames [FadeDistance,FoamHeight,FoamSpread,Power]}
  n31 = Subtract(A=TextureCoordinate(), B=@n30.FoamSpread)
  n32 = Clamp(None=LinearInterpolate(Alpha=ComponentMask(None=@n31) {r=true}) {constA=-1.5})
  n33 = Clamp(None=LinearInterpolate(Alpha=ComponentMask(None=@n31) {g=true}) {constA=-1.5})
  n34 = Add(A=TextureCoordinate(), B=@n30.FoamSpread)
  n35 = ComponentMask(None=@n34) {r=true}
  n36 = Clamp(None=LinearInterpolate(Alpha=OneMinus(None=@n35)) {constA=-1.5})
  n37 = ComponentMask(None=@n34) {g=true}
  n38 = Clamp(None=LinearInterpolate(Alpha=OneMinus(None=@n37)) {constA=-1.5})
  n39 = Clamp(None=LinearInterpolate(A=LinearInterpolate(A=@n32, B=@n33), B=LinearInterpolate(A=@n36, B=@n38)))
  n40 = Power(Base=OneMinus(None=@n39)) {constExponent=6}
  n41 = Multiply(A=ParticleColor().RGB, B=@n40)
  n42 = Multiply(A=@n41, B=ScalarParameter() {defaultValue=1, parameterName=EmissivDepthScale})
  n43 = DepthFade(FadeDistance=Multiply(A=DynamicParameter() {paramNames [FadeDistance,FoamHeight,FoamSpread,Power]}.FadeDistance) {constB=1.5})
  MaterialFunctionCall(Emissive=StaticSwitchParameter(True=LinearInterpolate(A=@n41, B=@n42, Alpha=@n43), False=@n41) {parameterName=UseDepthEmissive}) {materialFunction=&MF_FX_EmissiveAdjustments}
  n44 = ParticleColor()
  n45 = Multiply(A=VertexColor().R, B=@n29)
  n46 = DynamicParameter() {paramNames [FadeDistance,FoamHeight,FoamSpread,Power]}
  n47 = DepthFade(Opacity=Power(Base=Multiply(A=@n44.A, B=@n45), Exp=@n46.Power), FadeDistance=@n46.FadeDistance)
  n48 = MaterialFunctionCall(Distance=ScalarParameter() {defaultValue=100, parameterName=NearCameraFadeDistance}) {materialFunction=&Near_Camera_Fade}
  StaticSwitchParameter(True=Multiply(A=@n47, B=@n48.Result), False=@n47) {parameterName=Near Camera WPO}
  n49 = Multiply(A=Multiply(A=@n39, B=@n40)) {constB=10}
  n50 = SphereMask(A=TextureCoordinate(), B=Constant() {r=0.5}) {attenuationRadius=0.8, hardnessPercent=0.1}
  n51 = ScalarParameter() {defaultValue=0.8, parameterName=ScaleSphereMask}
  n52 = Add(A=@n26.RGB, B=StaticSwitchParameter(True=Power(Base=@n49), False=Multiply(A=@n50, B=@n51)) {parameterName=UseBoxRamp})
  n53 = Multiply(
   A=VectorParameter() {parameterName=OffsetHeight, defaultValue=<rgba 0 0 1 0>}.RGB
   B=DynamicParameter() {paramNames [FadeDistance,FoamHeight,FoamSpread,Power]}.FoamHeight
  )
  n54 = Transform(None=Multiply(A=Multiply(A=@n52, B=@n53), B=VertexColor().R)) {transformSourceType=TRANSFORMSOURCE_Local}
  StaticSwitchParameter(True=LinearInterpolate(B=@n54, Alpha=@n48.Result), False=@n54) {parameterName=Near Camera WPO}
 ]
}
