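# M_CubeWake is a translucent, lit mesh-particle water material, apparently for the cube's wake, with radial ripples.
# Dynamic-parameter-driven sine rings distort bubble and caustic textures and, with caustic height, drive vertex offset, with optional center push and Gerstner waves for whirlpools.
# Time-of-day colors, optionally varied by panning noise, combine with a view-space fake reflection and sphere-masked emissive.
# Particle alpha, depth fade and optional radial masks with center cutout shape opacity.
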Material /Game/Effects/Fort_Effects/Materials/Water/M_CubeWake
graph {
 properties {
  bEnableResponsiveAA=true
  bUsedWithMeshParticles=true
  bUsedWithNanite=true
  bUsedWithNiagaraMeshParticles=true
  blendMode=BLEND_Translucent
  pixelDepthOffsetMode=PDOM_Legacy
  translucencyLightingMode=TLM_VolumetricPerVertexNonDirectional
  usedShadingModels=Default Lit
 }
 nodes [
  n1 = Constant() {r=0.5}
  n2 = TextureCoordinate()
  n3 = DynamicParameter() {defaultValue=<rgba 0.2 1 1 1>, paramNames [Ripple, RippleScale, FadeDistance, ScaleCausticHeight]}
  n4 = Multiply(
   A=Power(Base=Distance(A=@n1, B=@n2)) {constExponent=1.4}
   B=LinearInterpolate(
    A=ScalarParameter() {defaultValue=3, parameterName=NumRipplesStart}
    B=ScalarParameter() {defaultValue=10, parameterName=NumRipplesEnd}
    Alpha=@n3.Ripple
   ) {constA=12, constB=3}
  ) {constB=2}
  n5 = Multiply(A=Multiply(A=@n3.Ripple) {constB=-1}, B=ScalarParameter() {defaultValue=3, parameterName=RippleSpeed}) {constB=-5}
  n6 = ConstantBiasScale(None=Sine(None=Add(A=@n4, B=@n5)))
  n7 = Multiply(A=ScalarParameter() {defaultValue=0.1, parameterName=UVDistort}, B=@n6)
  n8 = Add(A=Multiply(A=TextureCoordinate(), B=ScalarParameter() {defaultValue=5, parameterName=BubbleScale}), B=@n7)
  n9 = ReflectionVectorWS(CustomWorldNormal=TextureSample(UVs=@n8) {samplerType=SAMPLERTYPE_Normal, texture=/Game/Effects/Fort_Effects/Textures/Water/T_MaskeSplashes_Bubbles_N}.RGB)
  n10 = ConstantBiasScale(None=ComponentMask(None=Transform(None=@n9) {transformSourceType=TRANSFORMSOURCE_World, transformType=TRANSFORM_View}) {g=true, r=true})
  n11 = ComponentMask(None=VectorParameter() {parameterName=CubeMapOffset, defaultValue=<rgba 1 1 0 1>}.RGB) {g=true, r=true}
  n12 = Multiply(
   A=TextureSampleParameter2D(UVs=Multiply(A=@n10, B=@n11)) {parameterName=SphereReflectionMap, texture=/Game/Items/Art_noLOD/Textures/T_Items_MineralPowder_Glass_Reflection}.RGB
   B=ScalarParameter() {defaultValue=0.1, group=reflection, parameterName=Reflection Multiplier}
  )
  n13 = ComponentMask(None=MaterialFunctionCall() {materialFunction=/Engine/Functions/GetPostProcessFunctions/GetAmbientCubemapTint}.AmbientCubemapTint) {b=true, g=true, r=true}
  n14 = MaterialFunctionCall(
   Color_Morning=VectorParameter() {parameterName=WaterColorMorning, defaultValue=<rgba 0.3909 0.3217 0.5 1>}.RGB
   Color_Day=VectorParameter() {parameterName=WaterColorDay, defaultValue=<rgba 0.1779 0.3005 0.6308 1>}.RGB
   Color_Evening=VectorParameter() {parameterName=WaterColorEve, defaultValue=<rgba 0.008 0.0561 0.314 1>}.RGB
   Color_Night=VectorParameter() {parameterName=WaterColorNight, defaultValue=<rgba 0.3255 0.116 0.5573 1>}.RGB
  ) {materialFunction=&Color_Time_of_Day}
  n15 = Multiply(A=TextureCoordinate(), B=ScalarParameter() {defaultValue=3, parameterName=Color Variation Tile})
  n16 = Constant2Vector() {g=0.1, r=0.1}
  n17 = TextureObjectParameter() {parameterName=Color Noise Texture, texture=&T_ShieldCaustics}
  n18 = Panner(Coordinate=Add(A=@n15, B=Constant2Vector() {g=0.3548, r=0.4181}), Speed=Constant2Vector() {g=-0.1, r=0.1})
  n19 = Max(
   A=TextureSample(UVs=Panner(Coordinate=@n15, Speed=@n16), Tex=@n17) {texture=&T_ShieldCaustics}.R
   B=TextureSample(UVs=Multiply(A=@n18) {constB=0.7143}, Tex=@n17) {texture=&T_ShieldCaustics}.G
  )
  n20 = Panner(Coordinate=Add(A=@n15, B=Constant2Vector() {g=0.1484, r=0.8649}), Speed=Constant2Vector() {g=0.1, r=-0.1})
  n21 = Panner(Coordinate=Add(A=@n15, B=Constant2Vector() {g=0.7516, r=0.6513}), Speed=Constant2Vector() {g=-0.1, r=-0.1})
  n22 = Max(
   A=TextureSample(UVs=Multiply(A=@n20) {constB=1.186}, Tex=@n17) {texture=&T_ShieldCaustics}.R
   B=TextureSample(UVs=Multiply(A=@n21) {constB=1.1803}, Tex=@n17) {texture=&T_ShieldCaustics}.B
  )
  n23 = VectorParameter() {parameterName=Color Texture mask Options, channelNames {a="", b=Add, g=Multiply, r=Power}, defaultValue=<rgba 1 1 0 0>}
  n24 = Multiply(A=Power(Base=Max(A=@n19, B=@n22), Exp=@n23.Power), B=@n23.Multiply)
  n25 = LinearInterpolate(
   A=@n14.Result
   B=MaterialFunctionCall(
    Color_Morning=VectorParameter() {parameterName=WaterColorMorning 2, defaultValue=<rgba 0.1955 0.1608 0.25 1>}.RGB
    Color_Day=VectorParameter() {parameterName=WaterColorDay 2, defaultValue=<rgba 0.0889 0.1503 0.3154 1>}.RGB
    Color_Evening=VectorParameter() {parameterName=WaterColorEve 2, defaultValue=<rgba 0.004 0.0281 0.157 1>}.RGB
    Color_Night=VectorParameter() {parameterName=WaterColorNight 2, defaultValue=<rgba 0.1628 0.058 0.2786 1>}.RGB
   ) {materialFunction=&Color_Time_of_Day}.Result
   Alpha=Clamp(None=Add(A=@n24, B=@n23.Add))
  )
  n26 = ScalarParameter() {parameterName=Use Second Color Set}
  n27 = TextureCoordinate()
  n28 = ScalarParameter() {defaultValue=6, parameterName=Caustic_Scale}
  n29 = TextureSample(UVs=Add(A=Multiply(A=@n27, B=@n28), B=@n7)) {samplerType=SAMPLERTYPE_LinearColor, texture=/Game/Characters/Player/Male/Medium/Bodies/M_MED_SRDrift/Textures/T_ElectricNoise}
  n30 = TextureSample(UVs=@n8) {texture=/Game/Effects/Fort_Effects/Textures/Water/T_MaskeSplashes_Bubbles1_A}
  n31 = ScalarParameter() {defaultValue=2.25, parameterName=ScaleGlow}
  n32 = DynamicParameter() {defaultValue=<rgba 0.2 1 1 1>, paramNames [Ripple, RippleScale, FadeDistance, ScaleCausticHeight]}
  n33 = Multiply(A=@n32.FadeDistance) {constB=0.8}
  n34 = Multiply(
   A=MaterialFunctionCall(Hue Shift Percentage=ScalarParameter() {parameterName=HueShift}, Texture=LinearInterpolate(A=@n14.Result, B=@n25, Alpha=@n26)) {materialFunction=&HueShift}.Result
   B=LinearInterpolate(A=Add(A=@n29.G, B=@n30.RGB), B=Multiply(A=@n31, B=@n30.RGB) {constA=2}, Alpha=DepthFade(FadeDistance=@n33))
  )
  Add(A=Clamp(None=Multiply(A=@n12, B=@n13)), B=@n34)
  n35 = ParticleColor()
  n36 = Power(
   Base=SphereMask(A=TextureCoordinate(), B=Constant() {r=0.5}, Radius=ScalarParameter() {defaultValue=0.35, parameterName=Radius}) {attenuationRadius=0.35, hardnessPercent=1}
   Exp=ScalarParameter() {defaultValue=2, parameterName=SphereMaskContrast}
  )
  MaterialFunctionCall(Emissive=Multiply(A=@n35.RGB, B=Multiply(A=@n34, B=@n36))) {materialFunction=&MF_FX_EmissiveAdjustments}
  n37 = Multiply(A=@n29.G, B=@n32.ScaleCausticHeight)
  n38 = VectorParameter() {parameterName=RippleHeight, defaultValue=<rgba 0 0 2 0>}
  n39 = DynamicParameter() {defaultValue=<rgba 0.2 1 1 1>, paramNames [Ripple, RippleScale, FadeDistance, ScaleCausticHeight]}
  n40 = Transform(None=Multiply(A=Add(A=@n37, B=@n6), B=Multiply(A=@n38.RGB, B=@n39.RippleScale))) {transformSourceType=TRANSFORMSOURCE_Local}
  n41 = SphereMask(
   A=TextureCoordinate()
   B=Constant() {r=0.5}
   Radius=ScalarParameter() {defaultValue=0.5, parameterName=WPO_CenterRadius}
   Hardness=ScalarParameter() {parameterName=WPO_CenterHardness}
  )
  n42 = ScalarParameter() {parameterName=WPO_CenterPush}
  n43 = Add(A=@n40, B=Multiply(A=Multiply(A=@n41, B=@n42), B=Constant3Vector() {constant=<rgba 0 0 1 0>}))
  n44 = Multiply(
   A=MaterialFunctionCall!(
    Z Offset Only
    Position
    WaterBodyIndex=CollectionParameter() {collection=/Water/Materials/MPC/MPC_Water, parameterName=OceanWaterBodyIndex}
    Normals use VS Interpolation
    All Normals VS
   ) {materialFunction=/FortWater/Materials/Functions/MF_Gerstner_16_Waves}.WPO
   B=ScalarParameter() {defaultValue=1, parameterName=WaveWPO}
  )
  n45 = Constant()
  Reroute(None=Add(
    A=StaticSwitchParameter(True=@n43, False=@n40) {parameterName=Use Centeral Push on Mesh}
    B=StaticSwitchParameter(True=@n44, False=@n45) {parameterName=Use WPO For Whirlpool}
   ))
  ScalarParameter() {defaultValue=14, parameterName=Index}
  n46 = DepthFade(FadeDistance=@n32.FadeDistance)
  n47 = Multiply(A=@n35.A, B=ShadingPathSwitch(Default=@n46, Deferred=@n46, Mobile=Constant() {r=1}) {desc=Migrated from FeatureLevelSwitch})
  n48 = Power(
   Base=SphereMask(A=TextureCoordinate(), B=Constant() {r=0.5}, Radius=ScalarParameter() {defaultValue=0.35, parameterName=Alpha_Radius}) {attenuationRadius=0.35, hardnessPercent=1}
   Exp=ScalarParameter() {defaultValue=2, parameterName=Alpha_SphereMaskContrast}
  )
  n49 = SphereMask(
   A=TextureCoordinate()
   B=Constant() {r=0.5}
   Radius=ScalarParameter() {defaultValue=0.2, parameterName=Black_Center_Radius}
   Hardness=ScalarParameter() {defaultValue=0.2, parameterName=Black_Center_Hardness}
  ) {attenuationRadius=0.35, hardnessPercent=1}
  n50 = ScalarParameter() {defaultValue=0.5, parameterName=Black_Center_Alpha_Exp}
  n51 = Multiply(A=@n48, B=OneMinus(None=Power(Base=@n49, Exp=@n50)))
  StaticSwitchParameter(True=Multiply(A=@n47, B=StaticSwitchParameter(True=@n51, False=@n48) {parameterName=UseRemoveCenter?}), False=@n47) {parameterName=Use SphereMask Alpha}
  Time() {bOverride_Period=true, period=1}
 ]
}
