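# Translucent lit material for flowing jungle-mouth water, scrolling mainly along V.
# Two panning water normals are angle-blended, while a displacement texture panning with the second blends base color to foam and drives a second scalar blend.
# A UV-gradient mask combines vertical falloff, optional top edge fade and side fade broken up by scrolling noise.
# All panners share a CustomTime reroute: scaled time offset by vertex color red.
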Material /Game/Effects/Fort_Effects/Materials/Water/M_JungleMouthWater
graph {
 properties {
  blendMode=BLEND_Translucent
  ditherOpacityMask=true
  pixelDepthOffsetMode=PDOM_Legacy
  translucencyLightingMode=TLM_Surface
  usedShadingModels=Default Lit
 }
 nodes [
  n1 = ComponentMask(None=TextureCoordinate()) {g=true}
  n2 = Multiply(
   A=Power(Base=OneMinus(None=@n1), Exp=ScalarParameter() {defaultValue=1, parameterName=ScaleMask Power}) {constExponent=1}
   B=ScalarParameter() {defaultValue=8, parameterName=ScaleMask}
  )
  n3 = ScalarParameter() {defaultValue=2, parameterName=Top Edge Fade Power}
  n4 = Multiply(A=@n2, B=Multiply(A=Power(Base=@n1, Exp=@n3), B=ScalarParameter() {defaultValue=2, parameterName=Top Edge Fade Multi}))
  n5 = ScalarParameter() {parameterName=Use TopEdgeFade}
  n6 = ScalarParameter() {defaultValue=1, parameterName=Side Gradient Transition}
  n7 = Constant() {r=1}
  n8 = ComponentMask(None=TextureCoordinate()) {r=true}
  n9 = Multiply(A=Multiply(A=OneMinus(None=@n8), B=@n8)) {constB=4}
  n10 = TextureCoordinate()
  n11 = Multiply(A=ScalarParameter() {defaultValue=3, parameterName=EdgeMask_TileX}, B=ComponentMask(None=@n10) {r=true})
  n12 = Multiply(A=ComponentMask(None=@n10) {g=true}, B=ScalarParameter() {defaultValue=2, parameterName=EdgeMask})
  n13 = VectorParameter() {parameterName=EdgeMaskSpeed, defaultValue=<rgba 0.05 -2.85 0 0>}
  n14 = TextureSample(UVs=Panner(Coordinate=AppendVector(A=@n11, B=@n12), Time=NamedRerouteUsage(), Speed=ComponentMask(None=@n13.RGB) {g=true, r=true}) {bFractionalPart=true}) {texture=/Game/Global/Textures/Default/Noise/TankMoteNoise}
  n15 = Multiply(A=@n9, B=LinearInterpolate(Alpha=Add(A=@n9, B=@n14.G)) {constA=-0.5})
  Clamp(None=Multiply(A=LinearInterpolate(A=@n2, B=@n4, Alpha=@n5), B=If(A=@n6, "A > B"=@n15, "A == B"=@n15, "A < B"=@n7) {constB=0.99}))
  n16 = ScalarParameter() {defaultValue=2, parameterName=Water_N_2_TileX}
  n17 = TextureCoordinate()
  n18 = ComponentMask(None=@n17) {r=true}
  n19 = ComponentMask(None=@n17) {g=true}
  n20 = ScalarParameter() {defaultValue=0.5, parameterName=Water_N_2_TileY}
  n21 = Panner(
   Coordinate=AppendVector(A=Multiply(A=@n16, B=@n18), B=Multiply(A=@n19, B=@n20))
   Time=NamedRerouteUsage()
   Speed=ComponentMask(None=VectorParameter() {parameterName=Normal2Speed, defaultValue=<rgba 0.05 -1 0 0>}.RGB) {g=true, r=true}
  ) {bFractionalPart=true}
  n22 = TextureSample(UVs=@n21) {texture=/Game/Packages/DS_Fortnite_TRV_Terrain/Textures/T_TRV_HF_WaterDisp}
  LinearInterpolate(
   A=VectorParameter() {parameterName=BaseColor, defaultValue=<rgba 0.0132 0.0573 0.0078 0>}.RGB
   B=VectorParameter() {parameterName=FoamColor, defaultValue=<rgba 0.5903 0.6771 0.3263 1>}.RGB
   Alpha=@n22.G
  )
  LinearInterpolate(Alpha=@n22.G) {constA=0.05, constB=0.4}
  n23 = TextureCoordinate()
  n24 = AppendVector(
   A=Multiply(A=ScalarParameter() {defaultValue=1, parameterName=Water_N_1_TileX}, B=ComponentMask(None=@n23) {r=true})
   B=Multiply(A=ComponentMask(None=@n23) {g=true}, B=ScalarParameter() {defaultValue=0.8, parameterName=Water_N_1_TileY})
  )
  n25 = NamedRerouteUsage()
  n26 = ComponentMask(None=VectorParameter() {parameterName=NormalSpeed, defaultValue=<rgba 0.1 -1.5 0 0>}.RGB) {g=true, r=true}
  MaterialFunctionCall(
   BaseNormal=TextureSample(UVs=Panner(Coordinate=@n24, Time=@n25, Speed=@n26) {bFractionalPart=true}) {samplerType=SAMPLERTYPE_Normal, texture=/Game/Packages/DS_Fortnite_TRV_Terrain/Textures/T_TRV_HF_WaterNormal}.RGB
   AdditionalNormal=TextureSample(UVs=@n21) {samplerType=SAMPLERTYPE_Normal, texture=/Game/Packages/DS_Fortnite_TRV_Terrain/Textures/T_TRV_SwampWater_01_N}.RGB
  ) {materialFunction=&BlendAngleCorrectedNormals}
  n27 = ScalarParameter() {defaultValue=1, parameterName=Global Time Multiplier}
  n28 = Time()
  NamedRerouteDeclaration(None=Add(A=Multiply(A=@n27, B=@n28), B=VertexColor().R)) {name=CustomTime, nodeColor=<rgba 0 0.625 1 1>}
 ]
}
