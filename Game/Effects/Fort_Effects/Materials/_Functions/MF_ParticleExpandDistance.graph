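# Material function that scales up sprite particles far from the camera so effects stay visible under Temporal AA.
# It negates each vertex's offset from the particle center and multiplies it by MaxScaleAtDistance, producing an offset suited to World Position Offset that mirrors sprites as it enlarges them.
# An inverted, soft SphereMask of CameraDistance radius around the camera blends that offset in from zero as particles get farther away.
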MaterialFunction /Game/Effects/Fort_Effects/Materials/_Functions/MF_ParticleExpandDistance
graph {
 properties {
  bExposeToLibrary=true
  description="Scale up Sprites at specificed distances, good for when FX become invisible due to Temporal AA"
  libraryCategoriesText [Effects,Orion]
 }
 nodes [
  n1 = Multiply(A=Subtract(A=WorldPosition().XYZ, B=ParticlePositionWS())) {constB=-1}
  n2 = FunctionInput(Preview=Constant() {r=10}) {description=What is the maximum size of your sprites at camera distance, inputName=MaxScaleAtDistance, inputType=FunctionInput_Scalar}
  n3 = SphereMask(
   A=ParticlePositionWS()
   B=CameraPositionWS()
   Radius=FunctionInput(Preview=Constant() {r=256}) {description=Gets the distance between the camera and sprites, inputName=CameraDistance, inputType=FunctionInput_Scalar}
  ) {hardnessPercent=10}
  FunctionOutput(None=LinearInterpolate(B=Multiply(A=@n1, B=@n2), Alpha=OneMinus(None=@n3)))
 ]
}
