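# Default Lit sticky grenade goo material for Niagara mesh particles, with its output coming from the seasonal MF_LackLuster_Base function.
# The graph also holds goo chains that do not feed that output.
# These pan noise at a particle-driven speed for a flattened swirl normal, distorted particle-colored emissive, and sine-masked vertex-normal displacement scaled by a dynamic WPO parameter.
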Material /Game/Effects/Fort_Effects/Materials/Water/M_StickyGrenade_Goo
graph {
 properties {bUsedWithNiagaraMeshParticles=true, pixelDepthOffsetMode=PDOM_Legacy, usedShadingModels=Default Lit}
 nodes [
  BreakMaterialAttributes!(Attr=MaterialFunctionCall!(
    bBlendMaterialAttributes
    MA_In
    UVsOverride
    Override Normals
    BlendMaterialAttributesAlpha
    BaseColorOverride
    WorldPositionOverride
    VertexNormalOverride
    SourceNormalsStrengthOverride
   ) {materialFunction=/Game/TechArt/Seasonal/LackLuster/MaterialFunctions/MF_LackLuster_Base}.MA Out)
  n1 = TextureCoordinate()
  n2 = Constant() {r=1}
  n3 = Constant() {r=1.2}
  n4 = Panner(
   Coordinate=Multiply(A=@n1, B=AppendVector(A=@n2, B=@n3))
   Speed=MaterialFunctionCall(Y=DynamicParameter() {paramNames [WPO,PanSpeed,Param3,Param4]}.PanSpeed) {materialFunction=&MakeFloat2}.Result
  ) {speedY=-0.65}
  MaterialFunctionCall(
   Normal=TextureSample(UVs=@n4) {samplerType=SAMPLERTYPE_Normal, texture=/Game/Global/Textures/Default/Noise/T_SwirlNoise_N}.RGB
   Flatness=Constant() {r=0.5}
  ) {materialFunction=&FlattenNormal}
  n5 = TextureSample(UVs=@n4) {samplerType=SAMPLERTYPE_Grayscale, texture=&LowResBlurredNoise}
  n6 = Multiply(A=@n1) {constB=2}
  n7 = Multiply(
   A=TextureSample(UVs=Add(A=@n5.R, B=@n6) {constB=0.5}) {texture=&T_NoisePack_01}.RGB
   B=Constant3Vector() {constant=<rgba 1 0.2 1 0>}
  )
  MaterialFunctionCall(Emissive=Multiply(A=ParticleColor().RGB, B=Divide(A=@n7) {constB=10})) {materialFunction=&MF_FX_EmissiveAdjustments}
  n8 = MaterialFunctionCall(Float2=@n1) {materialFunction=&BreakOutFloat2Components}
  n9 = Saturate(None=Sine(None=Multiply(A=@n8.G) {constB=0.5}))
  n10 = TextureSample(UVs=@n4) {samplerType=SAMPLERTYPE_Grayscale, texture=&LowResBlurredNoise}
  Multiply(
   A=Multiply(A=VertexNormalWS(), B=Multiply(A=@n9, B=@n10.RGB))
   B=Multiply(A=ScalarParameter() {defaultValue=40, parameterName=WPO}, B=DynamicParameter() {paramNames [WPO,PanSpeed,Param3,Param4]}.WPO)
  )
  ScalarParameter() {defaultValue=1, parameterName=Metallic}
  ScalarParameter() {defaultValue=1, parameterName=Specular}
  ScalarParameter() {defaultValue=0.3, parameterName=Roughness}
  VectorParameter() {parameterName=Base Color, defaultValue=<rgba 0.6302 0.6302 0.6302 0>}
 ]
}
