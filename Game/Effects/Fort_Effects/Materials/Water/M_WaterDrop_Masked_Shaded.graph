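# Masked, lit material for water drop particle sprites (Cascade and Niagara), using a sphere normal and mask to make flat sprites read as round drops.
# Fresnel-weighted cubemap reflection and a sun/moon specular highlight are summed, tinted by particle color and hue-shifted, alongside a small particle-colored emissive term.
# The rotatable sphere mask cuts out drops moving above a speed threshold.
# Mobile replaces the reflection with a constant and skips the speed cutoff.
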Material /Game/Effects/Fort_Effects/Materials/Water/M_WaterDrop_Masked_Shaded
graph {
 properties {
  bUsedWithNiagaraSprites=true
  bUsedWithParticleSprites=true
  blendMode=BLEND_Masked
  pixelDepthOffsetMode=PDOM_Legacy
  usedShadingModels=Default Lit
 }
 nodes [
  n1 = ReflectionVectorWS() {bNormalizeCustomWorldNormal=true}
  n2 = Multiply(
   A=Fresnel(Normal=TextureSample() {samplerType=SAMPLERTYPE_Normal, texture=&T_Sphere_N}.RGB) {exponent=2}
   B=Multiply(
    A=TextureSampleParameterCube!(UVs=@n1, Apply View MipBias) {parameterName=MaterialExpressionTextureSampleParameterCube_0, texture=/Game/Packages/Fortress_Sky/TexturesHDR/T_hdrsky_test03_low}.RGB
    B=VectorParameter() {parameterName=Cube Map Multiplier, defaultValue=<rgba 1 1 1 0>}.RGB
   ) {constB=1.5}
  )
  n3 = ComponentMask(None=MaterialFunctionCall() {materialFunction=/Engine/Functions/GetPostProcessFunctions/GetAmbientCubemapTint}.AmbientCubemapTint) {b=true, g=true, r=true}
  n4 = ShadingPathSwitch(Default=Desaturation(None=Multiply(A=@n2, B=@n3), Fraction=Constant()), Mobile=Constant() {r=1}) {desc=Migrated from FeatureLevelSwitch}
  n5 = CollectionParameter() {
   collection=&FortniteMaterialParameters
   parameterName=SunAndMoonModelDirectionalVector
  }
  n6 = DotProduct(A=ComponentMask(None=@n5) {b=true, g=true, r=true}, B=ReflectionVectorWS())
  n7 = Multiply(
   A=Power(Base=Clamp(None=@n6), Exp=ScalarParameter() {defaultValue=15, parameterName=Spec Highlight Power}) {constExponent=20}
   B=ScalarParameter() {defaultValue=50, parameterName=Spec Intense}
  )
  n8 = ParticleColor()
  MaterialFunctionCall(
   Hue Shift Percentage=ScalarParameter() {defaultValue=0.1, parameterName="Hue Shift Percentage (S)"}
   Texture=Multiply(A=Add(A=@n4, B=@n7), B=@n8.RGB)
  ) {materialFunction=&HueShift}
  MaterialFunctionCall(Emissive=Multiply(A=@n8.RGB, B=ScalarParameter() {defaultValue=0.1, parameterName=Emissive})) {materialFunction=&MF_FX_EmissiveAdjustments}
  n9 = MaterialFunctionCall(UVs=TextureCoordinate(), "Rotation Angle (0-1)"=ScalarParameter() {parameterName="Rotation Angle (0-1) (S)"}) {materialFunction=&CustomRotator}
  n10 = ParticleSpeed()
  n11 = ScalarParameter() {defaultValue=1000, parameterName=ParticleSpeed_FadeAboveThisValue}
  n12 = LinearInterpolate(A=Constant() {r=1}, B=Constant(), Alpha=Clamp(None=Subtract(A=@n10, B=@n11)))
  n13 = Constant() {r=1}
  Reroute(None=Multiply(
    A=TextureSample(UVs=@n9.Rotated Values) {texture=/Game/Effects/Fort_Effects/Textures/Mask/T_Sphere_A}.G
    B=ShadingPathSwitch(Default=@n12, Mobile=@n13) {desc=Migrated from FeatureLevelSwitch}
   ))
  ScalarParameter() {defaultValue=1, parameterName=Specular}
  ScalarParameter() {parameterName=Metallic}
  ScalarParameter() {defaultValue=0.05, parameterName=Roughness}
  TextureSample(UVs=@n9.Rotated Values) {samplerType=SAMPLERTYPE_Normal, texture=&T_Sphere_N}
 ]
}
