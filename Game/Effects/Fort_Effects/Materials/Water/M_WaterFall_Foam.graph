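# Masked, lit foam material for waterfall effects on mesh particles, including Niagara.
# Two panning tile-texture samples blend by a dynamic parameter to distort splash-mask UVs.
# That mask and dynamic height control feed an optional temporally dithered opacity cutout.
# Switches add detail normals and Jiggle_WPO wobble, alongside a vertex-color-masked height offset driven by a particle-age cosine, with constant age on mobile.
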Material /Game/Effects/Fort_Effects/Materials/Water/M_WaterFall_Foam
graph {
 properties {
  bUsedWithMeshParticles=true
  bUsedWithNanite=true
  bUsedWithNiagaraMeshParticles=true
  blendMode=BLEND_Masked
  pixelDepthOffsetMode=PDOM_Legacy
  usedShadingModels=Default Lit
 }
 nodes [
  DynamicParameter() {defaultValue=<rgba 0 0 0 1>, paramNames [ZheightOffset, Unused, Tile, LerpTextures]}
  n1 = TextureSample(UVs=Panner(Coordinate=TextureCoordinate() {uTiling=3, vTiling=3})) {samplerType=SAMPLERTYPE_Normal, texture=/Game/Packages/Fortnite_UniqueMaterials_1/Textures/T_WaterFall_Tile_N}
  n2 = VertexNormalWS()
  Multiply(A=@n1.RGB, B=@n2)
  ParticleColor()
  ScalarParameter() {defaultValue=1, parameterName=Roughness}
  n3 = ScalarParameter() {defaultValue=1, parameterName=WPO_Speed}
  n4 = ScalarParameter() {defaultValue=0.5, parameterName=WPO_WaveScaleSM}
  n5 = ScalarParameter() {defaultValue=1, parameterName=WPO_WaveScale}
  StaticSwitchParameter(
   True=ComponentMask(None=MaterialFunctionCall!(Speed=@n3, SmallWaveScle=@n4, WaveScale=@n5) {materialFunction=/Game/Effects/Fort_Effects/Materials/_Functions/Jiggle_WPO}.Result) {b=true}
   False=Constant()
  ) {parameterName=UseWPO_Z}
  StaticSwitchParameter(True=Add(A=@n1.RGB, B=@n2), False=@n2) {parameterName=UseDetailNormal}
  n6 = DynamicParameter() {defaultValue=<rgba 0 0 0 1>, paramNames [ControlZHeight, ContrastHigh, Tile, LerpTextures]}
  n7 = Multiply(A=@n6.Tile, B=TextureCoordinate() {vTiling=0})
  n8 = TextureCoordinate() {uTiling=0}
  n9 = Multiply(A=TextureCoordinate(), B=ScalarParameter() {defaultValue=0.12, group=WaterFall_UVDistort, parameterName=UVDistortTilingA})
  n10 = Multiply(
   A=TextureSample(UVs=Panner(Coordinate=@n9) {speedY=0.125}) {texture=/Game/Packages/Fortnite_UniqueMaterials_1/Textures/T_WaterFall_Tile}.B
   B=ScalarParameter() {defaultValue=0.0163, group=WaterFall_UVDistort, parameterName=UVDistortAmountA}
  )
  n11 = Multiply(
   A=TextureCoordinate() {uTiling=0.5, vTiling=11}
   B=ScalarParameter() {defaultValue=0.2, group=WaterFall_UVDistort, parameterName=UVDistortTilingB}
  )
  n12 = Multiply(
   A=TextureSample(UVs=Panner(Coordinate=@n11) {speedY=0.25}) {texture=&T_WaterFall_Tile}.A
   B=ScalarParameter() {defaultValue=0.0125, group=WaterFall_UVDistort, parameterName=UVDistortAmountB}
  )
  n13 = TextureSample(UVs=Add(A=Add(A=@n7, B=@n8), B=LinearInterpolate(A=@n10, B=@n12, Alpha=@n6.LerpTextures))) {texture=/Game/Effects/Fort_Effects/Textures/Water/T_WaterFallSplash_Mask}
  StaticSwitchParameter(
   True=MaterialFunctionCall(Alpha Threshold=LinearInterpolate(A=@n6.ControlZHeight, B=@n6.ControlZHeight, Alpha=@n13.R) {constA=-4}) {materialFunction=&DitherTemporalAA}.Result
   False=Constant() {r=1}
  ) {parameterName=UseOpacityMask}
  n14 = ParticleRelativeTime()
  n15 = Constant()
  n16 = ConstantBiasScale(None=Cosine(None=ShadingPathSwitch(Default=@n14, Mobile=@n15) {desc=Migrated from FeatureLevelSwitch}))
  n17 = VectorParameter() {parameterName=ZheightAdd, defaultValue=<rgba 0 0 16 1>}
  Transform(None=Multiply(A=VertexColor().R, B=Multiply(A=@n16, B=@n17.RGB))) {transformSourceType=TRANSFORMSOURCE_World, transformType=TRANSFORM_Local}
  VectorParameter() {parameterName=BaseColor, defaultValue=<rgba 0.5 0.5 0.5 0>}
 ]
}
