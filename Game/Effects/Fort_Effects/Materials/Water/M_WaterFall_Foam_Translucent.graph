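# Translucent lit material for waterfall foam and splash mesh particles, with index-of-refraction distortion.
# A panned waterfall texture distorts splash-mask UVs, with U tiling and erosion driven by Dynamic Parameters.
# The eroded mask is shaped by a soft radial falloff, and with particle alpha drives a quality-switched refraction value.
# It also computes height offsets from vertex-color red and particle-age cosine, plus switchable Jiggle_WPO wave motion.
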Material /Game/Effects/Fort_Effects/Materials/Water/M_WaterFall_Foam_Translucent
graph {
 properties {
  bUsedWithMeshParticles=true
  bUsedWithNanite=true
  bUsedWithNiagaraMeshParticles=true
  blendMode=BLEND_Translucent
  pixelDepthOffsetMode=PDOM_Legacy
  refractionMethod=RM_IndexOfRefraction
  usedShadingModels=Default Lit
 }
 nodes [
  n1 = TextureCoordinate()
  ComponentMask(None=@n1) {r=true}
  DynamicParameter() {defaultValue=<rgba 1 1 3 1>, paramNames [Erode,Param2,Tile,LerpTextures]}
  MaterialFunctionCall() {materialFunction=&DebugTimeSine}
  n2 = ParticleColor()
  n3 = Constant() {r=0.5}
  n4 = SphereMask(A=@n1, B=@n3, Radius=@n3) {hardnessPercent=0}
  n5 = Multiply(A=Power(Base=@n4) {constExponent=3}) {constB=0.25}
  n6 = DynamicParameter() {defaultValue=<rgba 1 1 3 1>, paramNames [Erode,Param2,Tile,LerpTextures]}
  n7 = TextureCoordinate() {vTiling=0}
  n8 = Panner(Coordinate=Multiply(A=TextureCoordinate(), B=ScalarParameter() {defaultValue=0.12, group=WaterFall_UVDistort, parameterName=UVDistortTilingA})) {speedY=0.125}
  n9 = Add(
   A=Add(A=Multiply(A=@n6.Tile, B=@n7), B=TextureCoordinate() {uTiling=0})
   B=Multiply(
    A=TextureSample(UVs=@n8) {texture=/Game/Packages/Fortnite_UniqueMaterials_1/Textures/T_WaterFall_Tile}.B
    B=ScalarParameter() {defaultValue=0.0163, group=WaterFall_UVDistort, parameterName=UVDistortAmountA}
   )
  )
  n10 = Add(
   A=@n6.Erode
   B=LinearInterpolate(Alpha=TextureSample(UVs=@n9) {texture=/Game/Effects/Fort_Effects/Textures/Water/T_WaterFallSplash_Mask}.R) {constA=-1, constB=0}
  )
  n11 = Multiply(A=@n4, B=Clamp(None=Add(A=@n5, B=@n10)))
  QualitySwitch(Default=LinearInterpolate(Alpha=Multiply(A=@n2.A, B=@n11)) {constA=1, constB=1.1}, Low=Constant() {r=1})
  n12 = ScalarParameter() {defaultValue=1, parameterName=WPO_Speed}
  n13 = ScalarParameter() {defaultValue=0.5, parameterName=WPO_WaveScaleSM}
  n14 = ScalarParameter() {defaultValue=1, parameterName=WPO_WaveScale}
  StaticSwitchParameter(
   True=ComponentMask(None=MaterialFunctionCall!(Speed=@n12, SmallWaveScle=@n13, WaveScale=@n14) {materialFunction=/Game/Effects/Fort_Effects/Materials/_Functions/Jiggle_WPO}.Result) {b=true}
   False=Constant()
  ) {parameterName=UseWPO_Z}
  n15 = ParticleRelativeTime()
  n16 = Constant()
  n17 = ConstantBiasScale(None=Cosine(None=ShadingPathSwitch(Default=@n15, Mobile=@n16) {desc=Migrated from FeatureLevelSwitch}))
  n18 = VectorParameter() {parameterName=ZheightAdd, defaultValue=<rgba 0 0 16 1>}
  Transform(None=Multiply(A=VertexColor().R, B=Multiply(A=@n17, B=@n18.RGB))) {transformSourceType=TRANSFORMSOURCE_World, transformType=TRANSFORM_Local}
 ]
}
