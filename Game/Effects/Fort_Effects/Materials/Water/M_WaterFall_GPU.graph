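# Translucent, lit particle material for GPU waterfall sprites, usable by Niagara and legacy particle systems.
# A scalar switches particle RGB between direct color and an HSV offset on a fallback color, via MF_HSV_Offset.
# Particle alpha scales a round spark mask, and a sphere normal map supplies volume.
# The non-mobile speed-based lerp blends the mask with itself, so particle speed has no visible effect.
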Material /Game/Effects/Fort_Effects/Materials/Water/M_WaterFall_GPU
graph {
 properties {
  bEnableResponsiveAA=true
  bUsedWithNanite=true
  bUsedWithNiagaraSprites=true
  bUsedWithParticleSprites=true
  blendMode=BLEND_Translucent
  pixelDepthOffsetMode=PDOM_Legacy
  translucencyLightingMode=TLM_VolumetricPerVertexDirectional
  usedShadingModels=Default Lit
 }
 nodes [
  MaterialFunctionCall!(
   RGB=LinearInterpolate(
    A=ParticleColor().RGB
    B=VectorParameter() {desc=fallback for RGB only, parameterName=ParticleColor RGB fallback, defaultValue=<rgba 1 1 1 1>}.RGB
    Alpha=ScalarParameter() {parameterName=Use Particle Color RGB as HSV Offset}
   )
   HSV Offset=LinearInterpolate(A=Constant3Vector(), B=ParticleColor().RGB, Alpha=ScalarParameter() {parameterName=Use Particle Color RGB as HSV Offset})
   Optional Reference Color
  ) {materialFunction=/Game/Effects/Niagara/Emitters/Water/FoamEffect/MF_HSV_Offset}
  n1 = TextureSample() {texture=/Game/Global/Textures/Default/Masks_and_Gradients/T_SparkRound}
  n2 = ParticleSpeed()
  n3 = ScalarParameter() {defaultValue=100, parameterName=ParticleSpeed}
  n4 = LinearInterpolate(A=@n1.R, B=@n1.R, Alpha=Clamp(None=Divide(A=@n2, B=@n3)))
  Reroute(None=Multiply(A=ParticleColor().A, B=ShadingPathSwitch(Default=@n4, Mobile=@n1.R) {desc=Migrated from FeatureLevelSwitch}))
  Reroute(None=TextureSample() {samplerType=SAMPLERTYPE_Normal, texture=&T_Sphere_N}.RGB)
 ]
}
