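# Translucent, lit debug material for Niagara mesh particles, used to visualize snapping to the water surface.
# The WaterParticleDisplacement function supplies a world position offset, with a small constant Z lift added.
# A checker grid sampled by world XY at an adjustable scale is tinted by particle color.
# Particle alpha is multiplied by a UV-centered sphere mask with adjustable hardness, giving a round falloff.
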Material /Game/Effects/Fort_Effects/Materials/Water/M_WaterSnapToSurfaceDebug
graph {
 properties {
  bUsedWithNanite=true
  bUsedWithNiagaraMeshParticles=true
  blendMode=BLEND_Translucent
  pixelDepthOffsetMode=PDOM_Legacy
  usedShadingModels=Default Lit
 }
 nodes [
  Add(
   A=MaterialFunctionCall!(Ignore XY Offsets, Position, WaterBodyIndex, Depth) {materialFunction=/FortWater/Materials/Functions/WaterParticleDisplacement}.WPO
   B=Constant3Vector() {constant=<rgba 0 0 1 0>}
  )
  n1 = ParticleColor()
  Multiply(
   A=@n1.A
   B=SphereMask(
    A=TextureCoordinate()
    B=Constant() {r=0.5}
    Radius=Constant() {r=0.5}
    Hardness=ScalarParameter() {defaultValue=0.5, parameterName=SphereMaskHardness}
   )
  )
  n2 = ScalarParameter() {defaultValue=250, parameterName=TextureScale}
  Multiply(
   A=TextureSample(UVs=MaterialFunctionCall!(Scale=@n2) {materialFunction=/Engine/Functions/Engine_MaterialFunctions01/Coordinates/WorldPosition-XY}.Result) {texture=/Game/Maps/Textures/checkerGrid}.G
   B=@n1.RGB
  )
 ]
}
