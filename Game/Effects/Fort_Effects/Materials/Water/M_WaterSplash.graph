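# Unlit, two-sided translucent water splash material for mesh particles, tinted and faded by particle color.
# A panning cloud texture subtly distorts the water shape texture's UVs, with dynamic parameter TileEmissive tiling U.
# LerpAlpha blends the shape alpha toward a panning water pattern, and Param4 biases the remapped opacity.
# A local-space transform of vertex red and the Z parameter builds a vertical offset vector, apparently for mesh deformation.
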Material /Game/Effects/Fort_Effects/Materials/Water/M_WaterSplash
graph {
 properties {
  bUsedWithMeshParticles=true
  blendMode=BLEND_Translucent
  pixelDepthOffsetMode=PDOM_Legacy
  previewMesh=/Game/Effects/Fort_Effects/Mesh/Water/SM_SplashMesh_TileX1
  shadingModel=MSM_Unlit
  twoSided=true
  usedShadingModels=Unlit
 }
 nodes [
  n1 = ParticleColor()
  n2 = DynamicParameter() {defaultValue=<rgba 0 0 0 1>, paramNames [Z,LerpAlpha,TileEmissive,Param4]}
  n3 = Multiply(A=ComponentMask(None=TextureCoordinate()) {r=true}, B=@n2.TileEmissive)
  n4 = ComponentMask(None=TextureCoordinate() {uTiling=0}) {g=true}
  n5 = TextureCoordinate() {uTiling=1.5}
  n6 = Multiply(
   A=TextureSample(UVs=Panner(Coordinate=@n5) {speedY=0.5}) {texture=/Game/Effects/Fort_Effects/Textures/Tile/T_CloudScape_08}.R
   B=Constant() {r=0.0125}
  )
  n7 = TextureSample(UVs=Add(A=AppendVector(A=@n3, B=@n4), B=Multiply(A=@n6))) {texture=/Game/Effects/Fort_Effects/Textures/Water/T_WaterShapeTest}
  n8 = Multiply(A=Power(Base=@n7.RGB) {constExponent=4}, B=@n7.RGB)
  MaterialFunctionCall(Emissive=Multiply(A=@n1.RGB, B=Add(A=@n8, B=@n7.RGB))) {materialFunction=&MF_FX_EmissiveAdjustments}
  n9 = Constant() {r=-0.5}
  n10 = Constant() {r=1}
  n11 = Panner(Coordinate=TextureCoordinate() {uTiling=6, vTiling=3}) {speedY=0.75}
  n12 = DynamicParameter() {defaultValue=<rgba 0 0 0 1>, paramNames [Z,LerpAlpha,TileEmissive,Param4]}
  n13 = LinearInterpolate(
   A=@n7.A
   B=Multiply(A=TextureSample(UVs=@n11) {texture=/Game/Effects/Fort_Effects/Textures/Water/T_TilingWater_Pattern}.RGB, B=Multiply(A=@n8, B=@n7.A))
   Alpha=@n12.LerpAlpha
  )
  Multiply(A=LinearInterpolate(A=Add(A=@n9, B=@n2.Param4), B=Add(A=@n10, B=@n2.Param4), Alpha=ComponentMask(None=@n13) {r=true}), B=@n1.A) {constA=1, constB=0}
  n14 = VertexColor()
  n15 = Multiply(A=@n12.Z, B=Constant3Vector() {constant=<rgba 0 0 1 0>})
  Transform(None=Multiply(A=@n14.R, B=Add(A=@n14.R, B=@n15))) {transformSourceType=TRANSFORMSOURCE_Local}
 ]
}
