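# Masked, lit particle material for water splashes on GPU sprite and mesh particles, with a static switch for a blackbody lava variant.
# Random sub-UV cells feed the splash mask and a flattened normal.
# Color fakes reflection from a desaturated, Fresnel-weighted cubemap plus a sun/moon highlight, tinted by particle color and hue-shifted.
# The mask combines particle-alpha erosion, near-camera fade and high-speed cutoff; mobile replaces the cubemap term and cutoff with constants.
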Material /Game/Effects/Fort_Effects/Materials/Water/M_WaterSplash_Masked_GPU_Shaded
graph {
 properties {
  bUsedWithMeshParticles=true
  bUsedWithNiagaraSprites=true
  bUsedWithParticleSprites=true
  blendMode=BLEND_Masked
  pixelDepthOffsetMode=PDOM_Legacy
  translucencyLightingMode=TLM_VolumetricPerVertexDirectional
  usedShadingModels=Default Lit
 }
 nodes [
  n1 = TextureSampleParameterCube!(
   UVs=ReflectionVectorWS() {bNormalizeCustomWorldNormal=true}
   Level=ScalarParameter() {defaultValue=2, parameterName=Bias}
   Apply View MipBias
  ) {
   mipValueMode=TMVM_MipLevel
   parameterName=MaterialExpressionTextureSampleParameterCube_0
   texture=/Game/Packages/Fortress_Sky/TexturesHDR/T_hdrsky_test03_low
  }
  n2 = VectorParameter() {parameterName=Cube Map Multiplier, defaultValue=<rgba 15 15 15 0>}
  n3 = Clamp(None=Multiply(A=Fresnel() {exponent=2}, B=Multiply(A=@n1.RGB, B=@n2.RGB) {constB=1.5}))
  n4 = ComponentMask(None=MaterialFunctionCall() {materialFunction=/Engine/Functions/GetPostProcessFunctions/GetAmbientCubemapTint}.AmbientCubemapTint) {b=true, g=true, r=true}
  n5 = ShadingPathSwitch(
   Default=Desaturation(None=Multiply(A=@n3, B=@n4), Fraction=Constant() {r=1})
   Mobile=ScalarParameter() {defaultValue=0.01, parameterName=MobileCubeMapValue}
  ) {desc=Migrated from FeatureLevelSwitch}
  n6 = CollectionParameter() {
   collection=&FortniteMaterialParameters
   parameterName=SunAndMoonModelDirectionalVector
  }
  n7 = ComponentMask(None=@n6) {b=true, g=true, r=true}
  n8 = ReflectionVectorWS()
  n9 = Power(Base=Clamp(None=DotProduct(A=@n7, B=@n8)), Exp=ScalarParameter() {defaultValue=5, parameterName=Spec Highlight Power}) {constExponent=20}
  n10 = ScalarParameter() {defaultValue=25, parameterName=Spec Intense}
  n11 = Multiply(A=@n5, B=Clamp(None=Multiply(A=@n9, B=@n10)) {minDefault=0.7})
  n12 = Multiply(
   A=LinearInterpolate(B=Add(A=@n5, B=@n11), Alpha=ScalarParameter() {defaultValue=1, parameterName=Cube Map Intensity}) {constA=1}
   B=ParticleColor().RGB
  )
  n13 = ScalarParameter() {defaultValue=1, parameterName=ExponentIn}
  n14 = ScalarParameter() {defaultValue=1, parameterName=BaseReflectFractionIn}
  n15 = MaterialFunctionCall!(
   UseTexCoord 1
   Rows=ScalarParameter() {defaultValue=2, parameterName="Rows (S)"}
   Columns=ScalarParameter() {defaultValue=2, parameterName="Columns (S)"}
   Particle Random Value=ParticleRandom()
  ) {materialFunction=/Game/Effects/Fort_Effects/Materials/_Functions/MF_RandomSubUV_GPU_Sprites}
  n16 = ScalarParameter() {defaultValue=3, parameterName=NormalMipLevel}
  n17 = Normalize(VectorInput=MaterialFunctionCall(
    Normal=TextureSample!(UVs=@n15.Result, Tex, Level=@n16, Apply View MipBias) {mipValueMode=TMVM_MipLevel, samplerType=SAMPLERTYPE_Normal, texture=/Game/Effects/Fort_Effects/Textures/Water/T_waterSplash_N}.RGB
    Flatness=ScalarParameter() {defaultValue=0.75, parameterName="Flatness (S)"}
   ) {materialFunction=&FlattenNormal}.Result)
  Add(A=@n12, B=Multiply(A=@n5, B=Fresnel(ExponentIn=@n13, BaseReflectFractionIn=@n14, Normal=@n17)))
  n18 = ScalarParameter() {defaultValue=0.25, parameterName=Emissive}
  MaterialFunctionCall(Emissive=Multiply(A=StaticSwitchParameter(True=@n12, False=@n18) {group=LiqudType, parameterName=IsLava?}, B=@n18)) {materialFunction=&MF_FX_EmissiveAdjustments}
  n19 = TextureSample!(UVs=@n15.Result, Tex, Apply View MipBias) {samplerType=SAMPLERTYPE_LinearColor, texture=/Game/Effects/Fort_Effects/Textures/Water/T_waterSplash}
  n20 = Multiply(A=ScalarParameter() {parameterName=Temp}, B=Power(Base=@n19.R))
  Multiply(A=@n20, B=VectorParameter() {parameterName=LavaBaseColor, defaultValue=<rgba 0 0 0 1>}.RGB)
  n21 = ParticleColor()
  n22 = LinearInterpolate(A=@n19.R, B=@n19.R, Alpha=@n21.A)
  n23 = MaterialFunctionCall(Distance=ScalarParameter() {defaultValue=125, parameterName="Distance (S)"}) {materialFunction=&Near_Camera_Fade}
  n24 = Constant() {r=1}
  n25 = Constant()
  n26 = Clamp(None=Subtract(A=ParticleSpeed(), B=ScalarParameter() {defaultValue=500, parameterName=ParticleSpeed_FadeAboveThisValue}))
  Multiply(
   A=Multiply(
    A=MaterialFunctionCall!(Dynamic Param=@n21.A, Alpha=@n22) {materialFunction=/Game/Effects/Fort_Effects/Materials/_Functions/Erode_Alpha}.Result
    B=Clamp(None=@n23.Result)
   )
   B=ShadingPathSwitch(Default=LinearInterpolate(A=@n24, B=@n25, Alpha=@n26), Mobile=Constant() {r=1}) {desc=Migrated from FeatureLevelSwitch}
  )
  Reroute(None=@n17)
  ScalarParameter() {defaultValue=1, parameterName=Specular}
  ScalarParameter() {defaultValue=1, parameterName=Metallic}
  ScalarParameter() {defaultValue=1, parameterName=Roughness}
  n27 = MaterialFunctionCall(Hue Shift Percentage=ScalarParameter() {parameterName="Hue Shift Percentage (S)"}, Texture=@n12) {materialFunction=&HueShift}
  StaticSwitchParameter(True=Multiply(A=BlackBody!(Temp=@n20), B=@n27.Result), False=@n27.Result) {group=LiqudType, parameterName=IsLava?}
  StaticSwitchParameter() {group=LiqudType, parameterName=UseDistanceShrink?}
 ]
}
