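# Translucent lit material for Niagara mesh particles that draws an animated foamy water ripple tinted by time of day.
# Four panned swirl-foam samples combine to distort the V coordinate, which power curves shape into a band driving color and alpha.
# Alpha is shaped by edge gradients, erosion, depth fade, view-angle fade and vertex alpha.
# A panned, flattened normal map adds detail, and a static switch blends in water-body displacement normals and offset.
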Material /Game/Effects/Fort_Effects/Materials/Water/M_Water_Ripple_Mesh
graph {
 properties {bUsedWithNiagaraMeshParticles=true, blendMode=BLEND_Translucent, pixelDepthOffsetMode=PDOM_Legacy, usedShadingModels=Default Lit}
 nodes [
  Constant()
  n1 = Panner(Coordinate=TextureCoordinate() {uTiling=2, vTiling=0.25}) {speedY=0.1}
  n2 = MaterialFunctionCall!(Ignore XY Offsets, Position, WaterBodyIndex, Depth) {materialFunction=/FortWater/Materials/Functions/WaterParticleDisplacement}
  MaterialFunctionCall(
   BaseNormal=MaterialFunctionCall(
    Normal=TextureSample(UVs=@n1) {samplerType=SAMPLERTYPE_Normal, texture=/Game/Athena/Apollo/Water/Textures/02_Rounded/03_WaterFoam_Veroni_base_N}.RGB
    Flatness=ScalarParameter() {defaultValue=0.5, parameterName=NormalFlaten}
   ) {materialFunction=&FlattenNormal}.Result
   AdditionalNormal=StaticSwitchParameter(True=@n2.Normal, False=Constant3Vector() {constant=<rgba 0.5 0.5 1 0>}) {parameterName=UseWaterDisplacement}
  ) {materialFunction=&BlendAngleCorrectedNormals}
  Multiply()
  n3 = Multiply(A=ScalarParameter() {defaultValue=-0.1, parameterName=UVOffset}, B=Constant2Vector() {g=1})
  n4 = TextureCoordinate()
  n5 = Multiply(
   A=TextureCoordinate() {vTiling=0.1}
   B=AppendVector(A=ScalarParameter() {defaultValue=1, parameterName=XTile_01}, B=ScalarParameter() {defaultValue=1, parameterName=YTile_01})
  )
  n6 = AppendVector(A=ScalarParameter() {defaultValue=0.05, parameterName=X_Speed_01}, B=ScalarParameter() {defaultValue=0.25, parameterName=Y_Speed_01})
  n7 = OneMinus(None=TextureSample(UVs=Panner(Coordinate=@n5, Speed=@n6) {speedX=-0.025, speedY=0.25}) {samplerType=SAMPLERTYPE_LinearColor, texture=/Game/Global/Textures/Default/Caustics/03_WaterFoam_AccentSwirls_H}.G)
  n8 = TextureCoordinate() {uTiling=6, vTiling=0.5}
  n9 = AppendVector(A=ScalarParameter() {defaultValue=1, parameterName=XTile_02}, B=ScalarParameter() {defaultValue=1, parameterName=YTile_02})
  n10 = ScalarParameter() {defaultValue=0.05, parameterName=X_Speed_01}
  n11 = ScalarParameter() {defaultValue=0.25, parameterName=Y_Speed_01}
  n12 = TextureSample(UVs=Panner(Coordinate=Multiply(A=@n8, B=@n9), Speed=AppendVector(A=@n10, B=@n11)) {speedX=0.05, speedY=0.25}) {samplerType=SAMPLERTYPE_LinearColor, texture=&03_WaterFoam_AccentSwirls_H}
  n13 = ScalarParameter() {defaultValue=1, parameterName=XTile_03}
  n14 = ScalarParameter() {defaultValue=1, parameterName=YTile_03}
  n15 = Panner(
   Coordinate=Multiply(A=TextureCoordinate() {uTiling=6, vTiling=0.5}, B=AppendVector(A=@n13, B=@n14))
   Speed=AppendVector(A=ScalarParameter() {defaultValue=0.05, parameterName=X_Speed_03}, B=ScalarParameter() {defaultValue=0.25, parameterName=Y_Speed_03})
  ) {speedX=0.05, speedY=0.25}
  n16 = ScalarParameter() {defaultValue=1, parameterName=XTile_04}
  n17 = ScalarParameter() {defaultValue=1, parameterName=YTile_04}
  n18 = Panner(
   Coordinate=Multiply(A=TextureCoordinate() {uTiling=6, vTiling=0.5}, B=AppendVector(A=@n16, B=@n17))
   Speed=AppendVector(A=ScalarParameter() {defaultValue=0.05, parameterName=X_Speed_04}, B=ScalarParameter() {defaultValue=0.25, parameterName=Y_Speed_04})
  ) {speedX=0.05, speedY=0.25}
  n19 = Multiply(
   A=Add(A=Multiply(A=@n7) {constB=2}, B=OneMinus(None=@n12.G))
   B=Add(
    A=TextureSample(UVs=@n15) {samplerType=SAMPLERTYPE_LinearColor, texture=&03_WaterFoam_AccentSwirls_H}.G
    B=TextureSample(UVs=@n18) {samplerType=SAMPLERTYPE_LinearColor, texture=&03_WaterFoam_AccentSwirls_H}.G
   )
  )
  n20 = ScalarParameter() {defaultValue=1, parameterName=DistortionAmount}
  n21 = ComponentMask(None=Add(A=Add(A=@n3, B=@n4), B=Multiply(A=@n19, B=@n20))) {g=true}
  n22 = Power(Base=@n21) {constExponent=5}
  n23 = Multiply(A=Power(Base=OneMinus(None=@n21)) {constExponent=3}) {constB=5}
  n24 = Multiply(A=Multiply(A=Multiply(A=@n22, B=@n23)) {constB=25}, B=ScalarParameter() {defaultValue=1, parameterName=AlphaBoost}) {constB=0.5}
  n25 = Multiply(
   A=MaterialFunctionCall(
    Color_Morning=VectorParameter() {group=TOD, parameterName="Color_Morning (V3)", defaultValue=<rgba 1 1 1 0>}.RGB
    Color_Day=VectorParameter() {group=TOD, parameterName="Color_Day (V3)", defaultValue=<rgba 1 1 1 0>}.RGB
    Color_Evening=VectorParameter() {group=TOD, parameterName="Color_Evening (V3)", defaultValue=<rgba 1 1 1 0>}.RGB
    Color_Night=VectorParameter() {group=TOD, parameterName="Color_Night (V3)", defaultValue=<rgba 1 1 1 0>}.RGB
   ) {materialFunction=&Color_Time_of_Day}.Result
   B=@n24
  )
  n26 = MaterialFunctionCall(
   Value_Morning=ScalarParameter() {defaultValue=1, group=Opacity, parameterName="Value_Morning (S)"}
   Value_Day=ScalarParameter() {defaultValue=1, group=Opacity, parameterName="Value_Day (S)"}
   Value_Evening=ScalarParameter() {defaultValue=1, group=Opacity, parameterName="Value_Evening (S)"}
   Value_Night=ScalarParameter() {defaultValue=1, group=Opacity, parameterName="Value_Night (S)"}
  ) {materialFunction=&Scalar_Time_of_Day}
  Multiply(A=@n25, B=@n26.Result)
  n27 = MaterialFunctionCall() {materialFunction=&LinearGradient}
  n28 = ScalarParameter() {defaultValue=0.6, parameterName=EdgeMaskPower}
  n29 = Multiply(A=Power(Base=OneMinus(None=@n27.UGradient), Exp=@n28) {constExponent=0.5}, B=Power(Base=@n27.UGradient, Exp=@n28) {constExponent=0.5})
  n30 = ScalarParameter() {defaultValue=5, parameterName=EdgeMask}
  n31 = ScalarParameter() {parameterName=Erode}
  n32 = OneMinus(None=MaterialFunctionCall() {materialFunction=&LinearGradient}.VGradient)
  n33 = Saturate(None=Multiply(A=Saturate(None=@n24), B=Power(Base=@n32) {constExponent=0.75}))
  n34 = DepthFade(Opacity=Saturate(None=Add(A=@n31, B=@n33)), FadeDistance=ScalarParameter() {defaultValue=64, parameterName=DepthFadeDistance})
  n35 = MaterialFunctionCall("FadeSensitivity (optional)"=ScalarParameter() {defaultValue=2, parameterName=AngleFade}) {materialFunction=&FadeBasedOnViewAngle}
  Multiply(A=Multiply(A=Multiply(A=@n29, B=@n30), B=Multiply(A=@n34, B=@n35.Result)), B=VertexColor().A)
  Multiply(B=@n26.Result)
  Reroute(None=@n25)
  ScalarParameter() {defaultValue=1, parameterName=EmissiveMult}
  n36 = MaterialFunctionCall(Z=ScalarParameter() {parameterName=ZOffset}) {materialFunction=&MakeFloat3}
  StaticSwitchParameter(True=Multiply(A=Add(A=@n2.WPO, B=@n36.Result), B=Constant3Vector() {constant=<rgba 0.5 0.5 1 0>}), False=Constant()) {parameterName=UseWaterDisplacement}
 ]
}
