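# Translucent, lit mesh-particle material for waterfall foam, tinted and faded by particle color.
# A splash mask, warped by a dynamic-parameter blend of two panning distortion samples, feeds depth-faded opacity.
# Time-of-day functions scale color and emissive, and a water normal map is blended with the vertex normal.
# A red-vertex-color-masked offset vector follows a cosine of particle relative time, with a constant replacing time on mobile.
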Material /Game/Effects/Fort_Effects/Materials/Water/M_Waterfall_Foam_v2
graph {
 properties {
  bUsedWithMeshParticles=true
  blendMode=BLEND_Translucent
  pixelDepthOffsetMode=PDOM_Legacy
  translucencyLightingMode=TLM_VolumetricPerVertexDirectional
  usedShadingModels=Default Lit
 }
 nodes [
  n1 = ParticleColor()
  n2 = DynamicParameter() {defaultValue=<rgba 0 0 0 1>, paramNames [ControlZHeight, ContrastHigh, Tile, LerpTextures]}
  n3 = Multiply(A=@n2.Tile, B=TextureCoordinate() {vTiling=0})
  n4 = TextureCoordinate() {uTiling=0}
  n5 = Multiply(A=TextureCoordinate(), B=ScalarParameter() {defaultValue=0.125, group=WaterFall_UVDistort, parameterName=UVDistortTilingA})
  n6 = Multiply(
   A=TextureSample(UVs=Panner(Coordinate=@n5) {speedY=0.125}) {texture=/Game/Packages/Fortnite_UniqueMaterials_1/Textures/T_WaterFall_Tile}.B
   B=ScalarParameter() {defaultValue=0.0125, group=WaterFall_UVDistort, parameterName=UVDistortAmountA}
  )
  n7 = Multiply(
   A=TextureCoordinate() {uTiling=0.5, vTiling=11}
   B=ScalarParameter() {defaultValue=0.25, group=WaterFall_UVDistort, parameterName=UVDistortTilingB}
  )
  n8 = Multiply(
   A=TextureSample(UVs=Panner(Coordinate=@n7) {speedY=0.25}) {texture=&T_WaterFall_Tile}.A
   B=ScalarParameter() {defaultValue=0.025, group=WaterFall_UVDistort, parameterName=UVDistortAmountB}
  )
  n9 = TextureSample(UVs=Add(A=Add(A=@n3, B=@n4), B=LinearInterpolate(A=@n6, B=@n8, Alpha=@n2.LerpTextures))) {texture=/Game/Effects/Fort_Effects/Textures/Water/T_WaterFallSplash_Mask}
  DepthFade(
   Opacity=Multiply(A=@n1.A, B=LinearInterpolate(A=@n2.ControlZHeight, B=@n2.ControlZHeight, Alpha=@n9.R) {constA=-4})
   FadeDistance=ScalarParameter() {defaultValue=50, parameterName=DepthFadeDistance}
  )
  n10 = MaterialFunctionCall(
   Value_Morning=ScalarParameter() {defaultValue=0.4, parameterName=Morning}
   Value_Day=ScalarParameter() {defaultValue=0.75, parameterName=Day}
   Value_Evening=ScalarParameter() {defaultValue=0.5, parameterName=Evening}
   Value_Night=ScalarParameter() {defaultValue=0.85, parameterName=Night}
  ) {materialFunction=&Scalar_Time_of_Day}
  n11 = ScalarParameter() {defaultValue=0.05, parameterName=Morning_Emiss}
  n12 = ScalarParameter() {defaultValue=0.2, parameterName=Day_Emiss}
  n13 = ScalarParameter() {defaultValue=0.1, parameterName=Evening_Emiss}
  n14 = ScalarParameter() {defaultValue=0.075, parameterName=Night_Emiss}
  MaterialFunctionCall(Emissive=Multiply(
    A=Multiply(A=@n10.Result, B=@n1.RGB) {constB=0.75}
    B=MaterialFunctionCall(Value_Morning=@n11, Value_Day=@n12, Value_Evening=@n13, Value_Night=@n14) {materialFunction=&Scalar_Time_of_Day}.Result
   ) {constB=0.2}) {materialFunction=&MF_FX_EmissiveAdjustments}
  n15 = TextureCoordinate() {uTiling=3, vTiling=3}
  MaterialFunctionCall(
   BaseNormal=VertexNormalWS()
   AdditionalNormal=TextureSample(UVs=Panner(Coordinate=@n15)) {samplerType=SAMPLERTYPE_Normal, texture=/Game/Packages/Fortnite_UniqueMaterials_1/Textures/T_WaterFall_Tile_N}.RGB
  ) {materialFunction=&BlendAngleCorrectedNormals}
  n16 = ParticleRelativeTime()
  n17 = Constant()
  n18 = ConstantBiasScale(None=Cosine(None=ShadingPathSwitch(Default=@n16, Mobile=@n17) {desc=Migrated from FeatureLevelSwitch}))
  n19 = VectorParameter() {parameterName=ZheightAdd, defaultValue=<rgba 0 0 0 1>}
  Transform(None=Multiply(A=VertexColor().R, B=Multiply(A=@n18, B=@n19.RGB))) {transformSourceType=TRANSFORMSOURCE_World, transformType=TRANSFORM_Local}
 ]
}
